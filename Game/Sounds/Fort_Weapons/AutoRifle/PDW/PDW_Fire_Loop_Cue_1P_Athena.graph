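# First-person looping fire sound cue for the PDW auto rifle in Athena.
# A single looping wave player feeds a Modulator whose equal pitch bounds apply a fixed upward pitch shift rather than random variation.
# The cue boosts volume, uses first-person projectile attenuation and a first-person ranged-fire sound class, and joins the first-person gunfire concurrency group.
# Its debug flag is enabled.
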SoundCue /Game/Sounds/Fort_Weapons/AutoRifle/PDW/PDW_Fire_Loop_Cue_1P_Athena
root=Modulator_1
nodes [
 {id=Modulator_1, children [WavePlayer_6], properties {pitchMax=1.1, pitchMin=1.1}}
 {id=WavePlayer_6, properties {bLooping=true, soundWaveAssetPtr=&/PDW_Fire_Loop_01}}
]
properties {
 attenuationSettings=&Fort_Weapon_Projectile_FirstPerson_Athena_Att
 bDebug=true
 maxDistance=2097152
 soundClassObject=&Fort_WeaponRanged_Fire_1P_PSM
 volumeMultiplier=1.3
 concurrencySet [&Weapon_Gunfire_FirstPerson_Conc]
}
