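# First-person Sound Cue for the stop-fire sound of the PDW auto rifle in Athena.
# A single wave, PDW_Fire_Stop_01, plays through a Modulator into a three-input Mixer whose other two inputs are empty.
# The cue halves its volume, uses first-person projectile attenuation and a ranged-fire sound class, and shares first-person gunfire concurrency.
# Its debug flag is left on.
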SoundCue /Game/Sounds/Fort_Weapons/AutoRifle/PDW/PDW_Fire_Stop_Cue_1P_Athena
root=Mixer_17
nodes [
 {id=Mixer_17, children [Modulator_0,null,null], properties {inputVolume [1,0.7,0.75]}}
 {id=Modulator_0, children [WavePlayer_0]}
 {id=WavePlayer_0, properties {soundWaveAssetPtr=&/PDW_Fire_Stop_01}}
]
properties {
 attenuationSettings=&Fort_Weapon_Projectile_FirstPerson_Athena_Att
 bDebug=true
 maxDistance=2097152
 soundClassObject=&Fort_WeaponRanged_Fire_1P_PSM
 volumeMultiplier=0.5
 concurrencySet [&Weapon_Gunfire_FirstPerson_Conc]
}
