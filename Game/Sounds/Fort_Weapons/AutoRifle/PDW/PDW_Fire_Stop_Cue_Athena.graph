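# Sound cue for the PDW auto rifle's fire-stop sound in Athena, routed to the third-person ranged weapon fire sound class.
# A mixer layers a close-range wave with a quieter, volume-boosted medium-range wave, leaving its third input empty.
# Each layer runs through a modulator, its own projectile attenuation and a sound class override node.
# Local and global gunfire concurrency sets limit simultaneous playback.
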SoundCue /Game/Sounds/Fort_Weapons/AutoRifle/PDW/PDW_Fire_Stop_Cue_Athena
root=Mixer_17
nodes [
 {
  id=Attenuation_0
  children [Modulator_2]
  properties {attenuationSettings=&Fort_Weapon_Projectile_Medium_Mono_Athena_Att}
 }
 {
  id=Attenuation_49
  children [WavePlayer_0]
  properties {attenuationSettings=&Fort_Weapon_Projectile_Close_Athena_Att}
 }
 {id=Mixer_17, children [SoundClass_32, SoundClass_0, null], properties {inputVolume [1,0.15,0.75]}}
 {id=Modulator_0, children [Attenuation_49]}
 {id=Modulator_2, children [WavePlayer_9], properties {volumeMax=2, volumeMin=2}}
 {
  id=SoundClass_0
  children [Attenuation_0]
  properties {soundClassOverride=&Fort_WeaponRanged_Fire_3P_PSM}
 }
 {
  id=SoundClass_32
  children [Modulator_0]
  properties {soundClassOverride=&Fort_WeaponRanged_Fire_3P_PSM}
 }
 {id=WavePlayer_0, properties {soundWaveAssetPtr=&/PDW_Fire_Stop_01}}
 {id=WavePlayer_9, properties {soundWaveAssetPtr=&/PDW_Fire_Stop_MediumNew_01}}
]
properties {
 bDebug=true
 maxDistance=10579.2871
 soundClassObject=&Fort_WeaponRanged_Fire_3P_PSM
 volumeMultiplier=0.5
 concurrencySet [&Weapon_Gunfire_Local_Conc, &Weapon_Gunfire_Global_Conc]
}
