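# Sound cue for the final stage of the PDW auto rifle's reload.
# A single wave player feeds PDW_Reload_End_01 through a modulator at the root.
# It plays in the weapon foley sound class at about half volume with shared weapon reload attenuation.
# Its concurrency override prevents new instances from starting once the limit is reached.
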SoundCue /Game/Sounds/Fort_Weapons/AutoRifle/PDW/PDW_Reload_End_Cue
root=Modulator_1
nodes [
 {id=Modulator_1, children [WavePlayer_0]}
 {id=WavePlayer_0, properties {soundWaveAssetPtr=&/PDW_Reload_End_01}}
]
properties {
 attenuationSettings=&Fort_Weapon_Reload_Att
 bOverrideConcurrency=true
 maxDistance=2250
 soundClassObject=&Fort_WeaponFoley
 volumeMultiplier=0.501
 attenuationOverrides {falloffDistance=4100, attenuationShapeExtents {x=900}}
 concurrencyOverrides {resolutionRule=PreventNew}
}
