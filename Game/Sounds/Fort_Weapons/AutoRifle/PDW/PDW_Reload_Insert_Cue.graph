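# Sound cue for the magazine insert stage of the PDW auto rifle reload, classed as weapon foley.
# It layers the insert sound with a sub-mix of an immediate arm-reach sound and two slam waves delayed by half a second.
# The slams are pitch-shifted in opposite directions, one lower and one higher, and mixed at different levels.
# It uses reload attenuation and overrides concurrency to prevent new instances at the limit.
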SoundCue /Game/Sounds/Fort_Weapons/AutoRifle/PDW/PDW_Reload_Insert_Cue
root=Mixer_0
nodes [
 {id=Delay_0, children [Modulator_1], properties {delayMax=0.5, delayMin=0.5}}
 {id=Delay_2, children [Modulator_0], properties {delayMax=0.5, delayMin=0.5}}
 {id=Mixer_0, children [WavePlayer_1,Mixer_1], properties {inputVolume [1,1]}}
 {id=Mixer_1, children [Delay_0,WavePlayer_3,Delay_2], properties {inputVolume [0.6,0.5,0.2]}}
 {id=Modulator_0, children [WavePlayer_0], properties {pitchMax=1.2, pitchMin=1.2}}
 {id=Modulator_1, children [WavePlayer_2], properties {pitchMax=0.9, pitchMin=0.9}}
 {id=WavePlayer_0, properties {soundWaveAssetPtr=&/PDW_Reload_InsertSlam_02}}
 {id=WavePlayer_1, properties {soundWaveAssetPtr=&/PDW_Reload_Insert_01}}
 {id=WavePlayer_2, properties {soundWaveAssetPtr=&/PDW_Reload_InsertSlam_01}}
 {id=WavePlayer_3, properties {soundWaveAssetPtr=&/PDW_ArmReach_01}}
]
properties {
 attenuationSettings=&Fort_Weapon_Reload_Att
 bOverrideConcurrency=true
 maxDistance=2250
 soundClassObject=&Fort_WeaponFoley
 volumeMultiplier=0.501
 attenuationOverrides {falloffDistance=4100, attenuationShapeExtents {x=900}}
 concurrencyOverrides {resolutionRule=PreventNew}
}
